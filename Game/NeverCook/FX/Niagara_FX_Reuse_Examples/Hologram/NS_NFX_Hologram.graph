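# NS_NFX_Hologram is a Niagara system that shows a short-lived, mesh-based coin hologram, with a single camera-facing front-side emitter.
# A one-shot burst spawns one mesh when User.CoinSide is 0, lerping the count toward none as it approaches 1.
# Over each particle's life, curves grow then collapse the mesh's Y and Z scale, fade its alpha and drive dynamic material parameters.
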NiagaraSystem /Game/NeverCook/FX/Niagara_FX_Reuse_Examples/Hologram/NS_NFX_Hologram
dependencies {usedDataInterfaces [&NiagaraDataInterfaceCurve], usedDynamicInputs [], usedModules [&SystemState]}
emitters {CameraFacing_Coin_Hologram_FrontSide {
  emitterUpdateScript [
   &EmitterState()
   &SpawnBurst_Instantaneous(
    Spawn Count=&MakeIntFromFloat(Float=&Lerp_Float(A=1, B=0, Alpha=$User.CoinSide))
    Spawn Time=0.0417
    Use Spawn Probability=false
   )
  ]
  particleSpawnScript [&InitializeParticle(Mesh Scale Mode=::NewEnumerator3, Mesh Uniform Scale=10, UsePositionOffset=true, Write Material Random=true)]
  particleUpdateScript [
   &ParticleState()
   &SolveForcesAndVelocity(Clamp Velocity=false, Limit Acceleration=false)
   &ScaleMeshSize(Scale Factor=&MakeVector(
     X=1
     Y=&FloatFromCurve(FloatCurve=DataInterface() {propertyValues {Curve {keys [
          <at 0 0.5>
          <at 0.0833 1.0079>
          {arriveTangent=-1.096, interpMode=RCIM_Cubic, leaveTangent=-7.2979, tangentMode=RCTM_Break, time=0.8822, value=0.9489}
          <at 0.9513 0.0237>
         ]}}})
     Z=&FloatFromCurve(FloatCurve=DataInterface() {propertyValues {Curve {keys [
          <at 0 0.4955>
          <at 0.0955 1>
          {arriveTangent=-1.1021, interpMode=RCIM_Cubic, leaveTangent=0.4162, tangentMode=RCTM_Break, time=0.8964, value=1}
          {arriveTangent=-6.3724, interpMode=RCIM_Cubic, leaveTangent=-6.3724, tangentMode=RCTM_User, time=1.0011, value=0.002}
         ]}}})
    ))
   &Color(Scale Alpha=&FloatFromCurve(FloatCurve=DataInterface() {propertyValues {Curve {keys [
         <at -0.0046 1.0113>
         {arriveTangent=-4.1555, interpMode=RCIM_Cubic, leaveTangent=-4.1555, tangentMode=RCTM_User, time=0.8133, value=0.9117}
         <at 1 0>
        ]}}}))
   &DynamicMaterialParameters(
    Index 0 Param 1=&FloatFromCurve(FloatCurve=DataInterface() {propertyValues {Curve {keys [
         {arriveTangent=-0.6795, interpMode=RCIM_Cubic, leaveTangent=-3.6997, tangentMode=RCTM_User, time=0, value=1}
         {arriveTangent=-0.412, interpMode=RCIM_Cubic, leaveTangent=-0.412, tangentMode=RCTM_User, time=1.0091, value=0}
        ]}}})
    Index 0 Param 2=&FloatFromCurve(FloatCurve=DataInterface() {propertyValues {Curve {keys [<at -0.0046 0 RCIM_Cubic>,<at 0.996 -0.0333>]}}})
    Index 0 Param 3=&FloatFromCurve(
     FloatCurve=DataInterface() {propertyValues {Curve {keys [
         {arriveTangent=-2.2759, interpMode=RCIM_Cubic, leaveTangent=-2.2759, tangentMode=RCTM_User, time=-0.0046, value=0.4}
         {arriveTangent=-3.5698, interpMode=RCIM_Cubic, leaveTangent=-3.5698, tangentMode=RCTM_User, time=0.5673, value=0}
        ]}}}
     Scale Curve=0
    )
    Module.Index0Float0Write=true
    Module.Index0Float1Write=true
    Module.Index0Float2Write=true
    Module.Index0Float3Write=true
   )
  ]
  rendererClasses [&NiagaraMeshRendererProperties]
  renderers [{rendererClass=&NiagaraMeshRendererProperties, rendererIndex=0}]
 }}
summary {
 systemName=NS_NFX_Hologram
 emitters [{emitterName=CameraFacing_Coin_Hologram_FrontSide, rendererClasses [&NiagaraMeshRendererProperties]}]
 userVariables [User.CoinSide {type=NiagaraFloat, defaultValue=0, description=""}]
}
